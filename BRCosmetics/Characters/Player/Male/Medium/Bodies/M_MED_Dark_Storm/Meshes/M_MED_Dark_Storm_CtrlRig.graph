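# Control Rig for the M_MED_Dark_Storm player body that procedurally corrects collar, belt ring and buckle bones as the body moves.
# Clamped remaps turn head, left clavicle and thigh Euler angles into weights for chained Modify Transforms offsets.
# Buckle offsets are also weighted by their distance to the front belt rings.
# IsCrouchCtrl enables extra collar and left-buckle corrections; some nodes, including a Spherical Pose Reader, are unconnected.
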ControlRigBlueprint /BRCosmetics/Characters/Player/Male/Medium/Bodies/M_MED_Dark_Storm/Meshes/M_MED_Dark_Storm_CtrlRig
graphs [
 RigVMFunctionLibrary {links [], nodes []}
 RigVMModel {nodes [
   n1 = &RigVMFunction_MathQuaternionToEuler(
    RotationOrder=ZYX
    Value=&RigUnit_GetRelativeTransformForItem(Child=(Type=Bone,Name="head"), Parent=(Type=Bone,Name="spine_05"), bChildInitial=false, bParentInitial=false).RelativeTransform.Rotation
   )
   n2 = &RigVMFunction_MathQuaternionToEuler(
    RotationOrder=ZYX
    Value=&RigUnit_GetTransform(Item=(Type=Bone,Name="dyn_collar_back_r"), Space=LocalSpace, bInitial=true).Transform.Rotation
   )
   &RigUnit_ModifyTransforms(
    ExecutePin=()
    ItemToModify=((Item=(Type=Bone,Name="dyn_collar_back_r"),Transform=(Rotation=(X=0,Y=0,Z=0,W=1),Translation=(X=0,Y=0,Z=-2),Scale3D=(X=1,Y=1,Z=1))))
    Weight=&RigVMFunction_MathFloatRemap(
     SourceMaximum=-30
     SourceMinimum=-90
     TargetMaximum=0
     TargetMinimum=1
     bClamp=true
     Value=&RigVMFunction_MathFloatAdd(B=0, A=@n1.Result.Z) {template="Add::Execute(in A,in B,out Result)"}.Result
    ).Result
    ItemToModify.0.Transform.Rotation=&RigVMFunction_MathQuaternionFromEuler(Euler=/Script/RigVM.RigVMFunction_MathVectorAdd(B=(X=-90,Y=50,Z=-150), A=@n2.Result) {is=struct RigVMUnitNode, template="Add::Execute(in A,in B,out Result)"}.Result).Result
   )
   n3 = &RigUnit_GetRelativeTransformForItem(Child=(Type=Bone,Name="head"), Parent=(Type=Bone,Name="spine_05"), bChildInitial=false, bParentInitial=false)
   n4 = &RigVMFunction_MathFloatRemap(
    SourceMaximum=-50
    SourceMinimum=-90
    TargetMaximum=0
    TargetMinimum=1
    bClamp=true
    Value=&RigVMFunction_MathFloatAdd(B=0, A=&RigVMFunction_MathQuaternionToEuler(RotationOrder=ZYX, Value=@n3.RelativeTransform.Rotation).Result.Z) {template="Add::Execute(in A,in B,out Result)"}.Result
   )
   n5 = &RigVMFunction_MathQuaternionToEuler(
    RotationOrder=ZYX
    Value=&RigUnit_GetTransform(Item=(Type=Bone,Name="dyn_collar_back_r"), Space=LocalSpace, bInitial=true).Transform.Rotation
   )
   > &RigUnit_ModifyTransforms(
    ItemToModify=((Item=(Type=Bone,Name="dyn_collar_back_r"),Transform=(Rotation=(X=0,Y=0,Z=0,W=1),Translation=(X=2,Y=0,Z=6),Scale3D=(X=1,Y=1,Z=1))))
    Weight=@n4.Result
    ItemToModify.0.Transform.Rotation=&RigVMFunction_MathQuaternionFromEuler(Euler=&RigVMFunction_MathVectorAdd(B=(X=-90,Y=70,Z=-150), A=@n5.Result) {template="Add::Execute(in A,in B,out Result)"}.Result).Result
   )
   n6 = &RigVMFunction_MathQuaternionToEuler(
    RotationOrder=ZYX
    Value=&RigUnit_GetTransform(Item=(Type=Bone,Name="dyn_collar_back_r"), Space=LocalSpace, bInitial=true).Transform.Rotation
   )
   > &RigUnit_ModifyTransforms(
    ItemToModify=((Item=(Type=Bone,Name="dyn_collar_front_r"),Transform=(Rotation=(X=0,Y=0,Z=0,W=1),Translation=(X=0,Y=0,Z=0),Scale3D=(X=1,Y=1,Z=1))))
    Weight=@n4.Result
    ItemToModify.0.Transform.Rotation=&RigVMFunction_MathQuaternionFromEuler(Euler=&RigVMFunction_MathVectorAdd(B=(X=-90,Y=0,Z=-150), A=@n6.Result) {template="Add::Execute(in A,in B,out Result)"}.Result).Result
   )
   RigUnit_BeginExecution = &RigUnit_BeginExecution()
   n7 = &RigVMFunction_MathQuaternionToEuler(
    RotationOrder=ZYX
    Value=&RigUnit_GetRelativeTransformForItem(Child=(Type=Bone,Name="head"), Parent=(Type=Bone,Name="spine_05"), bChildInitial=false, bParentInitial=false).RelativeTransform.Rotation
   )
   n8 = &RigVMFunction_MathQuaternionToEuler(
    RotationOrder=ZYX
    Value=&RigUnit_GetTransform(Item=(Type=Bone,Name="dyn_collar_back_r"), Space=LocalSpace, bInitial=true).Transform.Rotation
   )
   &RigUnit_ModifyTransforms(
    ItemToModify=((Item=(Type=Bone,Name="dyn_collar_back_r"),Transform=(Rotation=(X=0,Y=0,Z=0,W=1),Translation=(X=0,Y=0,Z=0),Scale3D=(X=1,Y=1,Z=1))))
    Weight=&RigVMFunction_MathFloatRemap(
     SourceMaximum=-15
     SourceMinimum=-60
     TargetMaximum=0
     TargetMinimum=1
     bClamp=true
     Value=&RigVMFunction_MathFloatAdd(B=0, A=@n7.Result.Z) {template="Add::Execute(in A,in B,out Result)"}.Result
    ).Result
    ItemToModify.0.Transform.Rotation=&RigVMFunction_MathQuaternionFromEuler(Euler=&RigVMFunction_MathVectorAdd(B=(X=-90,Y=30,Z=-150), A=@n8.Result) {template="Add::Execute(in A,in B,out Result)"}.Result).Result
    ExecutePin=@RigUnit_BeginExecution.ExecutePin
   )
   n9 = &RigVMFunction_MathFloatAdd(A=&RigVMFunction_MathQuaternionToEuler(
     RotationOrder=ZYX
     Value=&RigUnit_GetRelativeTransformForItem(Child=(Type=Bone,Name="head"), Parent=(Type=Bone,Name="spine_05"), bChildInitial=false, bParentInitial=false).RelativeTransform.Rotation
    ).Result.Z) {template="Add::Execute(in A,in B,out Result)"}
   n10 = &RigVMFunction_MathFloatAdd(A=&RigVMFunction_MathQuaternionToEuler(
     RotationOrder=ZYX
     Value=&RigUnit_GetRelativeTransformForItem(Child=(Type=Bone,Name="head"), Parent=(Type=Bone,Name="spine_05"), bChildInitial=false, bParentInitial=false).RelativeTransform.Rotation
    ).Result.Y) {template="Add::Execute(in A,in B,out Result)"}
   n11 = &RigVMFunction_MathQuaternionToEuler(
    RotationOrder=ZYX
    Value=&RigUnit_GetTransform(Item=(Type=Bone,Name="dyn_collar_back_l"), Space=LocalSpace, bInitial=true).Transform.Rotation
   )
   > &RigUnit_ModifyTransforms(
    ItemToModify=((Item=(Type=Bone,Name="dyn_collar_back_l"),Transform=(Rotation=(X=0,Y=0,Z=0,W=1),Translation=(X=0,Y=0,Z=0),Scale3D=(X=1,Y=1,Z=1))))
    Weight=&RigVMFunction_MathFloatAdd(
     A=&RigVMFunction_MathFloatRemap(SourceMaximum=-10, SourceMinimum=-60, TargetMaximum=0, TargetMinimum=1, bClamp=true, Value=@n9.Result).Result
     B=&RigVMFunction_MathFloatRemap(SourceMaximum=0, SourceMinimum=-25, TargetMaximum=0, TargetMinimum=-1, bClamp=true, Value=@n10.Result).Result
    ) {template="Add::Execute(in A,in B,out Result)"}.Result
    ItemToModify.0.Transform.Rotation=&RigVMFunction_MathQuaternionFromEuler(Euler=&RigVMFunction_MathVectorAdd(B=(X=-90,Y=40,Z=-150), A=@n11.Result) {template="Add::Execute(in A,in B,out Result)"}.Result).Result
   )
   n12 = &RigVMFunction_MathQuaternionToEuler(
    RotationOrder=ZYX
    Value=&RigUnit_GetRelativeTransformForItem(Child=(Type=Bone,Name="head"), Parent=(Type=Bone,Name="spine_05"), bChildInitial=false, bParentInitial=false).RelativeTransform.Rotation
   )
   n13 = &RigVMFunction_MathQuaternionToEuler(
    RotationOrder=ZYX
    Value=&RigUnit_GetTransform(Item=(Type=Bone,Name="dyn_collar_front_r"), Space=LocalSpace, bInitial=true).Transform.Rotation
   )
   > &RigUnit_ModifyTransforms(
    ItemToModify=((Item=(Type=Bone,Name="dyn_collar_front_r"),Transform=(Rotation=(X=0,Y=0,Z=0,W=1),Translation=(X=3,Y=0,Z=0),Scale3D=(X=1,Y=1,Z=1))))
    Weight=&RigVMFunction_MathFloatRemap(
     SourceMaximum=-15
     SourceMinimum=-30
     TargetMaximum=0
     TargetMinimum=1
     bClamp=true
     Value=&RigVMFunction_MathFloatAdd(A=@n12.Result.Y) {template="Add::Execute(in A,in B,out Result)"}.Result
    ).Result
    ItemToModify.0.Transform.Rotation=&RigVMFunction_MathQuaternionFromEuler(Euler=&RigVMFunction_MathVectorAdd(B=(X=-90,Y=50,Z=-120), A=@n13.Result) {template="Add::Execute(in A,in B,out Result)"}.Result).Result
   )
   n14 = &RigVMFunction_MathQuaternionToEuler(
    RotationOrder=ZYX
    Value=&RigUnit_GetRelativeTransformForItem(Child=(Type=Bone,Name="head"), Parent=(Type=Bone,Name="spine_05"), bChildInitial=false, bParentInitial=false).RelativeTransform.Rotation
   )
   n15 = &RigVMFunction_MathQuaternionToEuler(
    RotationOrder=ZYX
    Value=&RigUnit_GetTransform(Item=(Type=Bone,Name="dyn_collar_back_r"), Space=LocalSpace, bInitial=true).Transform.Rotation
   )
   > &RigUnit_ModifyTransforms(
    ItemToModify=((Item=(Type=Bone,Name="dyn_collar_back_r"),Transform=(Rotation=(X=0,Y=0,Z=0,W=1),Translation=(X=0,Y=0,Z=-0.5),Scale3D=(X=1,Y=1,Z=1))))
    Weight=&RigVMFunction_MathFloatRemap(
     SourceMaximum=-5
     SourceMinimum=-25
     TargetMaximum=0
     TargetMinimum=1
     bClamp=true
     Value=&RigVMFunction_MathFloatAdd(A=@n14.Result.Y) {template="Add::Execute(in A,in B,out Result)"}.Result
    ).Result
    ItemToModify.0.Transform.Rotation=&RigVMFunction_MathQuaternionFromEuler(Euler=&RigVMFunction_MathVectorAdd(B=(X=100,Y=-90,Z=20), A=@n15.Result) {template="Add::Execute(in A,in B,out Result)"}.Result).Result
   )
   n16 = &RigUnit_GetRelativeTransformForItem(Child=(Type=Bone,Name="head"), Parent=(Type=Bone,Name="spine_05"), bChildInitial=false, bParentInitial=false)
   n17 = &RigVMFunction_MathFloatRemap(
    SourceMaximum=0
    SourceMinimum=15
    TargetMaximum=0
    TargetMinimum=1
    bClamp=true
    Value=&RigVMFunction_MathFloatAdd(A=&RigVMFunction_MathQuaternionToEuler(RotationOrder=ZYX, Value=@n16.RelativeTransform.Rotation).Result.Y) {template="Add::Execute(in A,in B,out Result)"}.Result
   )
   n18 = &RigVMFunction_MathQuaternionToEuler(
    RotationOrder=ZYX
    Value=&RigUnit_GetTransform(Item=(Type=Bone,Name="dyn_collar_back_r"), Space=LocalSpace, bInitial=true).Transform.Rotation
   )
   > &RigUnit_ModifyTransforms(
    ItemToModify=((Item=(Type=Bone,Name="dyn_collar_back_l"),Transform=(Rotation=(X=0,Y=0,Z=0,W=1),Translation=(X=0,Y=0,Z=-3),Scale3D=(X=1,Y=1,Z=1))))
    Weight=@n17.Result
    ItemToModify.0.Transform.Rotation=&RigVMFunction_MathQuaternionFromEuler(Euler=&RigVMFunction_MathVectorAdd(B=(X=100,Y=-80,Z=20), A=@n18.Result) {template="Add::Execute(in A,in B,out Result)"}.Result).Result
   )
   > &RigUnit_ModifyTransforms(
    ItemToModify=((Item=(Type=Bone,Name="dyn_collar_back_r"),Transform=(Rotation=(X=0,Y=0,Z=0,W=1),Translation=(X=0,Y=0,Z=-4),Scale3D=(X=1,Y=1,Z=1))))
    Weight=@n17.Result
   )
   n19 = &RigVMFunction_MathQuaternionToEuler(
    RotationOrder=ZYX
    Value=&RigUnit_GetRelativeTransformForItem(Child=(Type=Bone,Name="head"), Parent=(Type=Bone,Name="spine_05"), bChildInitial=false, bParentInitial=false).RelativeTransform.Rotation
   )
   n20 = &RigVMFunction_MathQuaternionToEuler(
    RotationOrder=ZYX
    Value=&RigUnit_GetTransform(Item=(Type=Bone,Name="dyn_collar_back_r"), Space=LocalSpace, bInitial=true).Transform.Rotation
   )
   > &RigUnit_ModifyTransforms(
    ItemToModify=((Item=(Type=Bone,Name="dyn_collar_back_r"),Transform=(Rotation=(X=0,Y=0,Z=0,W=1),Translation=(X=0,Y=0,Z=0),Scale3D=(X=1,Y=1,Z=1))))
    Weight=&RigVMFunction_MathFloatRemap(
     SourceMaximum=0
     SourceMinimum=-20
     TargetMaximum=0
     TargetMinimum=1
     bClamp=true
     Value=&RigVMFunction_MathFloatAdd(A=@n19.Result.Y) {template="Add::Execute(in A,in B,out Result)"}.Result
    ).Result
    ItemToModify.0.Transform.Rotation=&RigVMFunction_MathQuaternionFromEuler(Euler=&RigVMFunction_MathVectorAdd(B=(X=100,Y=-83,Z=20), A=@n20.Result) {template="Add::Execute(in A,in B,out Result)"}.Result).Result
   )
   n21 = &RigVMFunction_MathQuaternionToEuler(
    RotationOrder=ZYX
    Value=&RigUnit_GetRelativeTransformForItem(Child=(Type=Bone,Name="head"), Parent=(Type=Bone,Name="spine_05"), bChildInitial=false, bParentInitial=false).RelativeTransform.Rotation
   )
   n22 = &RigVMFunction_MathQuaternionToEuler(
    RotationOrder=ZYX
    Value=&RigUnit_GetTransform(Item=(Type=Bone,Name="dyn_collar_front_r"), Space=LocalSpace, bInitial=true).Transform.Rotation
   )
   > &RigUnit_ModifyTransforms(
    ItemToModify=((Item=(Type=Bone,Name="dyn_collar_front_l"),Transform=(Rotation=(X=0,Y=0,Z=0,W=1),Translation=(X=3,Y=0,Z=-2),Scale3D=(X=1,Y=1,Z=1))))
    Weight=&RigVMFunction_MathFloatRemap(
     SourceMaximum=-35
     SourceMinimum=-70
     TargetMaximum=0
     TargetMinimum=1
     bClamp=true
     Value=&RigVMFunction_MathFloatAdd(A=@n21.Result.X) {template="Add::Execute(in A,in B,out Result)"}.Result
    ).Result
    ItemToModify.0.Transform.Rotation=&RigVMFunction_MathQuaternionFromEuler(Euler=&RigVMFunction_MathVectorAdd(B=(X=-90,Y=45,Z=-90), A=@n22.Result) {template="Add::Execute(in A,in B,out Result)"}.Result).Result
   )
   n23 = &RigVMFunction_MathQuaternionToEuler(
    RotationOrder=ZYX
    Value=&RigUnit_GetRelativeTransformForItem(Child=(Type=Bone,Name="head"), Parent=(Type=Bone,Name="spine_05"), bChildInitial=false, bParentInitial=false).RelativeTransform.Rotation
   )
   n24 = &RigVMFunction_MathQuaternionToEuler(
    RotationOrder=ZYX
    Value=&RigUnit_GetTransform(Item=(Type=Bone,Name="dyn_collar_front_r"), Space=LocalSpace, bInitial=true).Transform.Rotation
   )
   > &RigUnit_ModifyTransforms(
    ItemToModify=((Item=(Type=Bone,Name="dyn_collar_front_r"),Transform=(Rotation=(X=0,Y=0,Z=0,W=1),Translation=(X=3,Y=0,Z=-2),Scale3D=(X=1,Y=1,Z=1))))
    Weight=&RigVMFunction_MathFloatRemap(
     SourceMaximum=70
     SourceMinimum=40
     TargetMinimum=0
     bClamp=true
     Value=&RigVMFunction_MathFloatAdd(A=@n23.Result.X) {template="Add::Execute(in A,in B,out Result)"}.Result
    ).Result
    ItemToModify.0.Transform.Rotation=&RigVMFunction_MathQuaternionFromEuler(Euler=&RigVMFunction_MathVectorAdd(B=(X=-90,Y=45,Z=-90), A=@n24.Result) {template="Add::Execute(in A,in B,out Result)"}.Result).Result
   )
   n25 = &RigVMFunction_MathQuaternionToEuler(
    RotationOrder=ZYX
    Value=&RigUnit_GetRelativeTransformForItem(Child=(Type=Bone,Name="head"), Parent=(Type=Bone,Name="spine_05"), bChildInitial=false, bParentInitial=false).RelativeTransform.Rotation
   )
   > &RigUnit_ModifyTransforms(
    ItemToModify=((Item=(Type=Bone,Name="dyn_collar_back_l"),Transform=(Rotation=(X=0,Y=0,Z=0,W=1),Translation=(X=0,Y=0,Z=0),Scale3D=(X=1,Y=1,Z=1))))
    Weight=&RigVMFunction_MathFloatRemap(
     SourceMaximum=15
     SourceMinimum=6
     TargetMinimum=0
     bClamp=true
     Value=&RigVMFunction_MathFloatAdd(A=@n25.Result.X) {template="Add::Execute(in A,in B,out Result)"}.Result
    ).Result
    ItemToModify.0.Transform.Rotation=&RigVMFunction_MathQuaternionFromEuler(Euler=(X=0,Y=10,Z=10)).Result
   )
   n26 = &RigVMFunction_MathQuaternionToEuler(
    RotationOrder=ZYX
    Value=&RigUnit_GetRelativeTransformForItem(Child=(Type=Bone,Name="clavicle_l"), Parent=(Type=Bone,Name="spine_05"), bChildInitial=false, bParentInitial=false).RelativeTransform.Rotation
   )
   > &RigUnit_ModifyTransforms(
    ItemToModify=((Item=(Type=Bone,Name="dyn_collar_front_l"),Transform=(Rotation=(X=0,Y=0,Z=0,W=1),Translation=(X=-5,Y=0,Z=0),Scale3D=(X=1,Y=1,Z=1))))
    Weight=&RigVMFunction_MathFloatRemap(
     SourceMaximum=10
     SourceMinimum=60
     TargetMinimum=0
     bClamp=true
     Value=&RigVMFunction_MathFloatAdd(A=@n26.Result.Y) {template="Add::Execute(in A,in B,out Result)"}.Result
    ).Result
    ItemToModify.0.Transform.Rotation=&RigVMFunction_MathQuaternionFromEuler(Euler=(X=0,Y=-25,Z=50)).Result
   )
   n27 = &RigVMFunction_MathQuaternionToEuler(
    RotationOrder=ZYX
    Value=&RigUnit_GetRelativeTransformForItem(Child=(Type=Bone,Name="thigh_l"), Parent=(Type=Bone,Name="pelvis"), bChildInitial=false, bParentInitial=false).RelativeTransform.Rotation
   )
   n28 = &RigVMFunction_MathQuaternionToEuler(
    RotationOrder=ZYX
    Value=&RigUnit_GetTransform(Item=(Type=Bone,Name="dyn_mstr_L_beltRing_front"), Space=LocalSpace, bInitial=true).Transform.Rotation
   )
   > &RigUnit_ModifyTransforms(
    ItemToModify=((Item=(Type=Bone,Name="dyn_mstr_L_beltRing_front"),Transform=(Rotation=(X=0,Y=0,Z=0,W=1),Translation=(X=0,Y=0,Z=0),Scale3D=(X=1,Y=1,Z=1))))
    Weight=&RigVMFunction_MathFloatRemap(
     SourceMaximum=0
     SourceMinimum=-90
     TargetMaximum=0
     TargetMinimum=1
     bClamp=true
     Value=&RigVMFunction_MathFloatAdd(B=0, A=@n27.Result.Z) {template="Add::Execute(in A,in B,out Result)"}.Result
    ).Result
    ItemToModify.0.Transform.Rotation=&RigVMFunction_MathQuaternionFromEuler(Euler=&RigVMFunction_MathVectorAdd(B=(X=-40,Y=-60,Z=-90), A=@n28.Result) {template="Add::Execute(in A,in B,out Result)"}.Result).Result
   )
   n29 = /Script/RigVM.RigVMFunction_MathFloatAbs(Value=&RigVMFunction_MathQuaternionToEuler(
     RotationOrder=ZYX
     Value=&RigUnit_GetRelativeTransformForItem(Child=(Type=Bone,Name="thigh_r"), Parent=(Type=Bone,Name="pelvis"), bChildInitial=false, bParentInitial=false).RelativeTransform.Rotation
    ).Result.Z) {is=struct RigVMUnitNode, template="Absolute::Execute(in Value,out Result)"}
   n30 = &RigVMFunction_MathQuaternionToEuler(
    RotationOrder=ZYX
    Value=&RigUnit_GetTransform(Item=(Type=Bone,Name="dyn_mstr_R_beltRing_front"), Space=LocalSpace, bInitial=true).Transform.Rotation
   )
   > &RigUnit_ModifyTransforms(
    ItemToModify=((Item=(Type=Bone,Name="dyn_mstr_R_beltRing_front"),Transform=(Rotation=(X=0,Y=0,Z=0,W=1),Translation=(X=0,Y=0,Z=0),Scale3D=(X=1,Y=1,Z=1))))
    Weight=&RigVMFunction_MathFloatRemap(
     SourceMaximum=180
     SourceMinimum=80
     TargetMaximum=0
     TargetMinimum=1
     bClamp=true
     Value=&RigVMFunction_MathFloatAdd(B=0, A=@n29.Result) {template="Add::Execute(in A,in B,out Result)"}.Result
    ).Result
    ItemToModify.0.Transform.Rotation=&RigVMFunction_MathQuaternionFromEuler(Euler=&RigVMFunction_MathVectorAdd(B=(X=-90,Y=-80,Z=60), A=@n30.Result) {template="Add::Execute(in A,in B,out Result)"}.Result).Result
   )
   n31 = &RigVMFunction_MathQuaternionToEuler(
    RotationOrder=ZYX
    Value=&RigUnit_GetRelativeTransformForItem(Child=(Type=Bone,Name="thigh_l"), Parent=(Type=Bone,Name="pelvis"), bChildInitial=false, bParentInitial=false).RelativeTransform.Rotation
   )
   > &RigUnit_ModifyTransforms(
    ItemToModify=((Item=(Type=Bone,Name="dyn_mstr_L_beltRing_front"),Transform=(Rotation=(X=0,Y=0,Z=0,W=1),Translation=(X=-5,Y=0,Z=0),Scale3D=(X=1,Y=1,Z=1))))
    Weight=&RigVMFunction_MathFloatRemap(
     SourceMaximum=-95
     SourceMinimum=-110
     TargetMaximum=0
     TargetMinimum=1
     bClamp=true
     Value=&RigVMFunction_MathFloatAdd(B=0, A=@n31.Result.Z) {template="Add::Execute(in A,in B,out Result)"}.Result
    ).Result
   )
   n32 = &RigUnit_GetTransform(Item=(Type=Bone,Name="dyn_mstr_L_beltRing_front"))
   n33 = &RigUnit_GetTransform(Item=(Type=Bone,Name="dyn_beltbuckle_l"))
   > &RigUnit_ModifyTransforms(
    ItemToModify=((Item=(Type=Bone,Name="dyn_beltbuckle_l"),Transform=(Rotation=(X=0,Y=0,Z=0,W=1),Translation=(X=-5,Y=0,Z=-5),Scale3D=(X=1,Y=1,Z=1))))
    WeightMinimum=0
    Weight=&RigVMFunction_MathFloatRemap(
     SourceMaximum=13
     SourceMinimum=16
     TargetMaximum=0
     TargetMinimum=1
     bClamp=true
     Value=&RigVMFunction_MathVectorDistance(A=@n32.Transform.Translation, B=@n33.Transform.Translation).Result
    ).Result
   )
   n34 = &RigUnit_GetTransform(Item=(Type=Bone,Name="dyn_beltbuckle_r"))
   n35 = &RigUnit_GetTransform(Item=(Type=Bone,Name="dyn_mstr_R_beltRing_front"))
   > n36 = &RigUnit_ModifyTransforms(
    ItemToModify=((Item=(Type=Bone,Name="dyn_beltbuckle_r"),Transform=(Rotation=(X=0,Y=0,Z=0,W=1),Translation=(X=5,Y=0,Z=3),Scale3D=(X=1,Y=1,Z=1))))
    WeightMinimum=0
    Weight=&RigVMFunction_MathFloatRemap(
     SourceMaximum=13
     SourceMinimum=16
     TargetMaximum=0
     TargetMinimum=1
     bClamp=true
     Value=&RigVMFunction_MathVectorDistance(B=@n34.Transform.Translation, A=@n35.Transform.Translation).Result
    ).Result
   )
   Branch_RigVMUnitNode = &RigVMFunction_ControlFlowBranch(ExecuteContext=@n36.ExecutePin, Condition=RigVMVariableNode() {variable=IsCrouchCtrl}.Value)
   n37 = &RigVMFunction_MathQuaternionToEuler(
    RotationOrder=ZYX
    Value=&RigUnit_GetRelativeTransformForItem(Child=(Type=Bone,Name="head"), Parent=(Type=Bone,Name="spine_05"), bChildInitial=false, bParentInitial=false).RelativeTransform.Rotation
   )
   &RigUnit_ModifyTransforms(
    ItemToModify=((Item=(Type=Bone,Name="dyn_collar_back_r"),Transform=(Rotation=(X=0,Y=0,Z=0,W=1),Translation=(X=5,Y=0,Z=5),Scale3D=(X=1,Y=1,Z=1))))
    ExecutePin=@Branch_RigVMUnitNode.True
    Weight=&RigVMFunction_MathFloatRemap(
     SourceMaximum=-60
     SourceMinimum=-90
     TargetMaximum=0
     TargetMinimum=1
     bClamp=true
     Value=&RigVMFunction_MathFloatAdd(A=@n37.Result.Z) {template="Add::Execute(in A,in B,out Result)"}.Result
    ).Result
    ItemToModify.0.Transform.Rotation=&RigVMFunction_MathQuaternionFromEuler(Euler=(X=0,Y=55,Z=0)).Result
   )
   n38 = &RigVMFunction_MathQuaternionToEuler(
    RotationOrder=ZYX
    Value=&RigUnit_GetRelativeTransformForItem(Child=(Type=Bone,Name="head"), Parent=(Type=Bone,Name="spine_05"), bChildInitial=false, bParentInitial=false).RelativeTransform.Rotation
   )
   > &RigUnit_ModifyTransforms(
    ItemToModify=((Item=(Type=Bone,Name="dyn_collar_back_r"),Transform=(Rotation=(X=0,Y=0,Z=0,W=1),Translation=(X=1,Y=0,Z=0),Scale3D=(X=1,Y=1,Z=1))))
    Weight=&RigVMFunction_MathFloatRemap(
     SourceMaximum=0
     SourceMinimum=-20
     TargetMaximum=0
     TargetMinimum=1
     bClamp=true
     Value=&RigVMFunction_MathFloatAdd(A=@n38.Result.Y) {template="Add::Execute(in A,in B,out Result)"}.Result
    ).Result
    ItemToModify.0.Transform.Rotation=&RigVMFunction_MathQuaternionFromEuler(Euler=(X=0,Y=15,Z=20)).Result
   )
   n39 = &RigVMFunction_MathQuaternionToEuler(
    RotationOrder=ZYX
    Value=&RigUnit_GetRelativeTransformForItem(Child=(Type=Bone,Name="thigh_l"), Parent=(Type=Bone,Name="pelvis"), bChildInitial=false, bParentInitial=false).RelativeTransform.Rotation
   )
   > &RigUnit_ModifyTransforms(
    ItemToModify=((Item=(Type=Bone,Name="dyn_beltbuckle_l"),Transform=(Rotation=(X=0,Y=0,Z=0,W=1),Translation=(X=5,Y=0,Z=-5),Scale3D=(X=1,Y=1,Z=1))))
    WeightMinimum=0
    Weight=&RigVMFunction_MathFloatRemap(
     SourceMaximum=-70
     SourceMinimum=-100
     TargetMaximum=0
     TargetMinimum=1
     bClamp=true
     Value=&RigVMFunction_MathFloatAdd(B=0, A=@n39.Result.Z) {template="Add::Execute(in A,in B,out Result)"}.Result
    ).Result
    ItemToModify.0.Transform.Rotation=&RigVMFunction_MathQuaternionFromEuler(Euler=(X=-20,Y=0,Z=50)).Result
   )
   &RigUnit_GetRelativeTransformForItem(Child=(Type=Bone,Name="dyn_mstr_L_beltRing_front"), Parent=(Type=Bone,Name="dyn_beltbuckle_l"), bChildInitial=false, bParentInitial=false)
   &RigVMFunction_MathVectorAdd(
    B=(X=-90,Y=120,Z=-90)
    A=&RigVMFunction_MathQuaternionToEuler(
     RotationOrder=ZYX
     Value=&RigUnit_GetTransform(Item=(Type=Bone,Name="dyn_collar_front_r"), Space=LocalSpace, bInitial=true).Transform.Rotation
    ).Result
   ) {template="Add::Execute(in A,in B,out Result)"}
   &RigVMFunction_MathVectorAdd(
    B=(X=0,Y=90,Z=-50)
    A=&RigVMFunction_MathQuaternionToEuler(
     RotationOrder=ZYX
     Value=&RigUnit_GetTransform(Item=(Type=Bone,Name="dyn_collar_back_l"), Space=LocalSpace, bInitial=true).Transform.Rotation
    ).Result
   ) {template="Add::Execute(in A,in B,out Result)"}
   &RigUnit_SphericalPoseReader(
    ActiveRegionScaleFactors=(PositiveWidth=1,NegativeWidth=1,PositiveHeight=1,NegativeHeight=1)
    ActiveRegionSize=0.1
    Debug=(bDrawDebug=True,DebugScale=25,DebugSegments=20,DebugThickness=0.25)
    DriverAxis=(X=-1,Y=0,Z=0)
    DriverItem=(Type=Bone,Name="dyn_beltbuckle_l")
    ExecutePin=()
    FalloffRegionScaleFactors=(PositiveWidth=1,NegativeWidth=1,PositiveHeight=1,NegativeHeight=1)
    FalloffSize=0.2
    OptionalParentItem=(Type=Bone)
    RotationOffset=(X=15,Y=-55,Z=0)
   )
  ]}
]
hierarchy [
 root {loc=<0 0 0>, rot=<0 0 0 1>}
 attach {parent=root, loc=<0 0 0>, rot=<0 0 0 1>}
 pelvis {parent=root, loc=<0 -0.7071 93.1191>, rot=<0.0127 0.707 -0.0127 -0.707>}
 spine_01 {parent=pelvis, loc=<6.7494 -0.6388 0>, rot=<0 0 0.034 -0.9994>}
 spine_02 {parent=spine_01, loc=<7.2518 -0.035 0>, rot=<0 0 0.0238 -0.9997>}
 spine_03 {parent=spine_02, loc=<6.0104 0 0>, rot=<0 0 0.0149 -0.9999>}
 spine_04 {parent=spine_03, loc=<8.1024 0 0>, rot=<0 0 -0.0132 -0.9999>}
 spine_05 {parent=spine_04, loc=<10.0146 0 0>, rot=<0 0 -0.0015 -1>}
 clavicle_l {parent=spine_05, loc=<4.3575 -0.1466 -2.3223>, rot=<0.0465 -0.7289 -0.009 -0.6829>}
 upperarm_l {parent=clavicle_l, loc=<15.5111 0 0>, rot=<0.0226 -0.3234 -0.0088 -0.9459>}
 lowerarm_l {parent=upperarm_l, loc=<26.6514 0 0>, rot=<0.0391 -0.0099 -0.2453 -0.9686>}
 hand_l {parent=lowerarm_l, loc=<22.3491 0 0>, rot=<-0.7081 -0.0507 0.0975 -0.6975>}
 index_metacarpal_l {parent=hand_l, loc=<1.6535 0.0183 -2.96>, rot=<0.0158 -0.0124 -0.0173 -0.9996>}
 index_01_l {parent=index_metacarpal_l, loc=<8.3008 0.0357 -0.0524>, rot=<-0.0044 0.0485 0.0924 -0.9945>}
 index_02_l {parent=index_01_l, loc=<3.3587 -0.3004 0.0452>, rot=<-0.002 0.0027 0.3225 -0.9466>}
 index_03_l {parent=index_02_l, loc=<2.5578 -0.123 -0.0252>, rot=<-0.002 0.0067 0.2558 -0.9667>}
 middle_metacarpal_l {parent=hand_l, loc=<2.1856 -0.1725 -0.5984>, rot=<0.006 0.0168 -0.0329 -0.9993>}
 middle_01_l {parent=middle_metacarpal_l, loc=<7.3275 -0.2447 0.0129>, rot=<-0.0208 0.0042 0.0858 -0.9961>}
 middle_02_l {parent=middle_01_l, loc=<4.0123 0.1401 0.1175>, rot=<-0.0218 0.0258 0.3512 -0.9357>}
 middle_03_l {parent=middle_02_l, loc=<2.6077 0.0534 0.1373>, rot=<0.0082 0.0188 0.1708 -0.9851>}
 pinky_metacarpal_l {parent=hand_l, loc=<1.7616 -0.9753 3.3814>, rot=<-0.0397 0.0814 -0.069 -0.9935>}
 pinky_01_l {parent=pinky_metacarpal_l, loc=<7.1709 -0.4831 -0.1094>, rot=<-0.0654 -0.0035 0.2242 -0.9723>}
 pinky_02_l {parent=pinky_01_l, loc=<2.8274 0.0893 -0.1043>, rot=<0.0037 0.0011 0.2854 -0.9584>}
 pinky_03_l {parent=pinky_02_l, loc=<2.2046 0.0307 -0.0078>, rot=<-0.0002 0.0098 0.1445 -0.9895>}
 ring_metacarpal_l {parent=hand_l, loc=<2.2074 -0.3625 1.3638>, rot=<-0.0284 0.0722 -0.0458 -0.9959>}
 ring_01_l {parent=ring_metacarpal_l, loc=<7.0295 -0.6197 -0.1446>, rot=<-0.0489 -0.0252 0.1892 -0.9804>}
 ring_02_l {parent=ring_01_l, loc=<3.6174 0.4647 -0.1027>, rot=<0 0 0.2806 -0.9598>}
 ring_03_l {parent=ring_02_l, loc=<2.4152 0.114 0.0199>, rot=<0 0 0.1189 -0.9929>}
 thumb_01_l {parent=hand_l, loc=<1.3118 -0.3191 -2.459>, rot=<0.6546 -0.4042 0.0664 -0.6354>}
 thumb_02_l {parent=thumb_01_l, loc=<5.9336 -0.0448 0.0401>, rot=<-0.0037 0.0213 0.298 -0.9543>}
 thumb_03_l {parent=thumb_02_l, loc=<2.2223 0.3682 -0.0907>, rot=<0.0118 -0.013 0.0066 -0.9998>}
 weapon_l {parent=hand_l, loc=<-0.1974 -3.4111 0.3811>, rot=<0 0 0 1>}
 lowerarm_twist_01_l {parent=lowerarm_l, loc=<17.8549 0 0>, rot=<0 0 0 1>}
 lowerarm_twist_02_l {parent=lowerarm_l, loc=<9.6199 0 0>, rot=<0 0 0 1>}
 upperarm_twist_01_l {parent=upperarm_l, loc=<0.0041 0 0>, rot=<0 0 0 1>}
 upperarm_twist_02_l {parent=upperarm_l, loc=<13.7941 0 0>, rot=<0 0 0 1>}
 clavicle_r {parent=spine_05, loc=<4.3575 -0.1466 2.3223>, rot=<0.7289 0.0465 -0.6829 0.009>}
 upperarm_r {parent=clavicle_r, loc=<-15.5111 0 0>, rot=<0.0226 -0.3234 -0.0088 -0.9459>}
 lowerarm_r {parent=upperarm_r, loc=<-26.6515 0 0>, rot=<0.0391 -0.0099 -0.2453 -0.9686>}
 hand_r {parent=lowerarm_r, loc=<-22.3492 0 0>, rot=<-0.7081 -0.0507 0.0975 -0.6975>}
 index_metacarpal_r {parent=hand_r, loc=<-1.6534 -0.0181 2.9601>, rot=<0.0158 -0.0124 -0.0173 -0.9996>}
 index_01_r {parent=index_metacarpal_r, loc=<-8.3008 -0.0357 0.0524>, rot=<-0.0044 0.0485 0.0924 -0.9945>}
 index_02_r {parent=index_01_r, loc=<-3.3589 0.3003 -0.0453>, rot=<-0.002 0.0027 0.3225 -0.9466>}
 index_03_r {parent=index_02_r, loc=<-2.5578 0.1231 0.0251>, rot=<-0.002 0.0067 0.2558 -0.9667>}
 middle_metacarpal_r {parent=hand_r, loc=<-2.1856 0.1726 0.5984>, rot=<0.006 0.0168 -0.0329 -0.9993>}
 middle_01_r {parent=middle_metacarpal_r, loc=<-7.3273 0.2447 -0.0129>, rot=<-0.0208 0.0042 0.0858 -0.9961>}
 middle_02_r {parent=middle_01_r, loc=<-4.0142 -0.1411 -0.07>, rot=<-0.0217 0.0259 0.3512 -0.9357>}
 middle_03_r {parent=middle_02_r, loc=<-2.6074 -0.0534 -0.1312>, rot=<0.0083 0.0188 0.1708 -0.9851>}
 pinky_metacarpal_r {parent=hand_r, loc=<-1.7617 0.9753 -3.3814>, rot=<-0.0397 0.0814 -0.069 -0.9935>}
 pinky_01_r {parent=pinky_metacarpal_r, loc=<-7.1707 0.4832 0.1094>, rot=<-0.0654 -0.0035 0.2242 -0.9723>}
 pinky_02_r {parent=pinky_01_r, loc=<-2.8274 -0.0892 0.1043>, rot=<0.0037 0.0011 0.2854 -0.9584>}
 pinky_03_r {parent=pinky_02_r, loc=<-2.2045 -0.031 0.0078>, rot=<-0.0002 0.0098 0.1445 -0.9895>}
 ring_metacarpal_r {parent=hand_r, loc=<-2.2073 0.3627 -1.3638>, rot=<-0.0284 0.0722 -0.0458 -0.9959>}
 ring_01_r {parent=ring_metacarpal_r, loc=<-7.0297 0.6197 0.1446>, rot=<-0.0489 -0.0252 0.1892 -0.9804>}
 ring_02_r {parent=ring_01_r, loc=<-3.6171 -0.4648 0.1027>, rot=<0 0 0.2806 -0.9598>}
 ring_03_r {parent=ring_02_r, loc=<-2.4153 -0.1141 -0.0199>, rot=<0 0 0.1189 -0.9929>}
 thumb_01_r {parent=hand_r, loc=<-1.3119 0.3192 2.459>, rot=<0.6546 -0.4042 0.0664 -0.6354>}
 thumb_02_r {parent=thumb_01_r, loc=<-5.9335 0.0447 -0.0401>, rot=<-0.0037 0.0213 0.298 -0.9543>}
 thumb_03_r {parent=thumb_02_r, loc=<-2.2223 -0.3682 0.0907>, rot=<0.0118 -0.013 0.0066 -0.9998>}
 weapon_r {parent=hand_r, loc=<0.1975 3.4112 -0.3811>, rot=<0 0 0 -1>}
 lowerarm_twist_01_r {parent=lowerarm_r, loc=<-17.855 0 0>, rot=<0 0 0 1>}
 lowerarm_twist_02_r {parent=lowerarm_r, loc=<-9.6199 0 0>, rot=<0 0 0 1>}
 upperarm_twist_01_r {parent=upperarm_r, loc=<-0.0041 0 0>, rot=<0 0 0 1>}
 upperarm_twist_02_r {parent=upperarm_r, loc=<-13.7942 0 0>, rot=<0 0 0 1>}
 neck_01 {parent=spine_05, loc=<11.3138 0.1529 0>, rot=<0 0 -0.2003 -0.9797>}
 neck_02 {parent=neck_01, loc=<4.5627 0 0>, rot=<0 0 0.0504 -0.9987>}
 head {parent=neck_02, loc=<5.3914 0 0>, rot=<0 0 0.115 -0.9934>}
 dyn_collar_front_l {parent=spine_05, loc=<13.9588 1.6183 -7.4002>, rot=<0.0284 -0.7065 -0.0284 -0.7065>}
 dyn_collar_back_l {parent=spine_05, loc=<16.0582 -4.8698 -4.5396>, rot=<0.3668 -0.6045 0.3172 -0.632>}
 dyn_collar_back_r {parent=spine_05, loc=<16.0582 -4.8698 4.54>, rot=<0.632 -0.3172 0.6045 -0.3669>}
 dyn_collar_front_r {parent=spine_05, loc=<13.9588 1.6183 7.4>, rot=<0.7065 0.0284 0.7065 -0.0284>}
 thigh_l {parent=pelvis, loc=<-2.5693 -0.0109 -7.7501>, rot=<0.0012 0.0572 0.0155 -0.9982>}
 calf_l {parent=thigh_l, loc=<-39.0484 0 0>, rot=<0.0006 -0.0161 -0.0391 -0.9991>}
 calf_twist_01_l {parent=calf_l, loc=<-35.813 -0.0967 0.0397>, rot=<0.0237 0.0095 -0.0035 -0.9997>}
 calf_twist_02_l {parent=calf_l, loc=<-18.8064 -0.0967 0.0397>, rot=<0.0237 0.0095 -0.0035 -0.9997>}
 foot_l {parent=calf_l, loc=<-42.561 0.1077 0.4292>, rot=<0.0418 -0.0455 0.0343 -0.9975>}
 ball_l {parent=foot_l, loc=<-6.1599 13.8264 0.6806>, rot=<0.0001 -0.0001 -0.7063 -0.7079>}
 thigh_twist_01_l {parent=thigh_l, loc=<-21.6303 0 0>, rot=<0.0013 0 0 -1>}
 dyn_beltbuckle_l {parent=thigh_l, loc=<-6.682 8.3625 -5.0082>, rot=<0.1027 -0.9168 0.2529 -0.2915>}
 thigh_r {parent=pelvis, loc=<-2.5691 -0.0109 7.7501>, rot=<0.0572 -0.0012 0.9982 0.0155>}
 calf_r {parent=thigh_r, loc=<39.0487 0 0>, rot=<0.0006 -0.0161 -0.0391 -0.9991>}
 calf_twist_01_r {parent=calf_r, loc=<35.8132 0.0967 -0.0397>, rot=<0.0237 0.0095 -0.0035 -0.9997>}
 calf_twist_02_r {parent=calf_r, loc=<18.8064 0.0967 -0.0397>, rot=<0.0237 0.0095 -0.0035 -0.9997>}
 foot_r {parent=calf_r, loc=<42.5612 -0.1077 -0.4292>, rot=<0.0418 -0.0455 0.0343 -0.9975>}
 ball_r {parent=foot_r, loc=<6.1598 -13.8264 -0.6806>, rot=<0.0001 -0.0001 -0.7063 -0.7079>}
 thigh_twist_01_r {parent=thigh_r, loc=<21.6303 0 0>, rot=<0.0013 0 0 -1>}
 dyn_beltbuckle_r {parent=thigh_r, loc=<6.6822 -8.3625 5.0081>, rot=<0.1027 -0.9168 0.2529 -0.2915>}
 dyn_mstr_L_beltRing_front {parent=pelvis, loc=<2.9031 10.7977 -5.796>, rot=<-0.2732 -0.3954 -0.561 -0.6741>}
 dyn_L_beltRing_front2 {parent=dyn_mstr_L_beltRing_front, loc=<0 0 0>, rot=<0 0 0 1>}
 dyn_mstr_R_beltRing_front {parent=pelvis, loc=<2.9031 10.7977 5.796>, rot=<0.707 -0.0127 -0.707 -0.0127>}
 dyn_L_beltRing_front {parent=dyn_mstr_R_beltRing_front, loc=<0 0 0>, rot=<0.5729 -0.4002 -0.266 -0.6639>}
 dyn_mstr_R_beltRing_back {parent=pelvis, loc=<5.4723 -12.1217 4.1436>, rot=<0.707 -0.0127 -0.707 -0.0127>}
 dyn_R_beltRing_back {parent=dyn_mstr_R_beltRing_back, loc=<0 0 0>, rot=<-0.5338 -0.3469 0.359 -0.6825>}
 dyn_mstr_L_beltRing_back {parent=pelvis, loc=<5.4722 -12.1217 -4.1436>, rot=<0.3527 -0.3532 0.546 -0.6729>}
 dyn_R_beltRing_back1 {parent=dyn_mstr_L_beltRing_back, loc=<0 0 0>, rot=<0 0 0 1>}
 ik_foot_root {parent=root, loc=<0 0 0>, rot=<0 0 0 1>}
 ik_foot_l {parent=ik_foot_root, loc=<15.2608 -4.4176 9.4433>, rot=<0.0357 0.7045 0.0284 -0.7082>}
 ik_foot_r {parent=ik_foot_root, loc=<-15.2609 -4.4176 9.4431>, rot=<0.7082 0.0284 -0.7045 0.0357>}
 ik_hand_root {parent=root, loc=<0 0 0>, rot=<0 0 0 1>}
 ik_hand_gun {parent=ik_hand_root, loc=<-51.9729 4.0884 102.6935>, rot=<0.6084 0.1793 0.4613 -0.6205>}
 ik_hand_l {parent=ik_hand_gun, loc=<53.0286 82.1782 35.2051>, rot=<-0.5102 -0.7906 -0.3387 0>}
 ik_hand_r {parent=ik_hand_gun, loc=<0 0 0>, rot=<0 0 0 1>}
]
variables [IsCrouchCtrl {type=bool, default=False}]
