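# Reduced physics asset for the M_MED_Dark_Storm male body cosmetic, covering dangling left-side belt and belt-ring bones.
# Capsules on the pelvis, lower spine and thighs form the body core, with free angular constraints to the pelvis.
# Capsule and box belt rings hang directly from the pelvis with limited or locked axes.
# Two belt bodies branch from a pair of tiny chained hip-pivot capsules, with twist locked on both belts.
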PhysicsAsset /BRCosmetics/Characters/Player/Male/Medium/Bodies/M_MED_Dark_Storm/Meshes/M_MED_Dark_Storm_Physics_Reduced
bodies {
 dyn_L_beltRing_back [{
   shapeType=Capsule
   length=7.9793
   radius=2.4159
   shapeName=dyn_L_beltRing_back_capsule
   center=<0.1172 -0.0062 -6.4179>
   rotation=<rot -3.2488 -89.5822 8.0698>
  }]
 dyn_L_beltRing_front [{
   shapeType=Box
   extentX=6.5729
   extentY=4.4817
   extentZ=27.0237
   shapeName=dyn_R_beltRing_front_box
   center=<1.896 -0.0602 -9.8663>
   rotation=<rot -0.5181 -90 -7.4>
  }]
 dyn_L_belt_B [{
   shapeType=Capsule
   length=11.5113
   radius=3.4609
   shapeName=dyn_L_belt_D_capsule
   center=<-8.2223 -0.0672 -1.079>
   rotation=<rot 0 92.5435 -76.5999>
  }]
 dyn_L_belt_D [{
   shapeType=Box
   extentX=5.0418
   extentY=4.7519
   extentZ=17.5677
   shapeName=dyn_R_belt_B_box
   center=<-8.5023 -0.3731 4.5216>
   rotation=<rot -4.1883 -87.9622 -51.8099>
  }]
 dyn_L_hip_pivot_D [{shapeType=Capsule, length=0.505, radius=0.505, shapeName=dyn_L_hip_pivot_D_capsule, rotation=<rot 0 -90 0>}]
 dyn_L_hip_pivot_D_02 [{shapeType=Capsule, length=0.505, radius=0.505, shapeName=dyn_L_hip_pivot_D_02_capsule, rotation=<rot 0 -90 0>}]
 pelvis [{shapeType=Capsule, length=18.25, radius=11.25, shapeName=pelvis_capsule, center=<-3.8653 -0.2934 0.0003>}]
 spine_01 [{shapeType=Capsule, length=7.209, radius=9.9729, shapeName=spine_01_capsule, center=<3.4989 0.5274 0>, rotation=<rot 90 0 0>}]
 thigh_l [{shapeType=Capsule, length=35, radius=11, shapeName=thigh_l_capsule, center=<-12.7438 -1.1739 1.617>, rotation=<rot 83.0415 0 0>}]
 thigh_r [{shapeType=Capsule, length=35, radius=11, shapeName=thigh_l_capsule, center=<18.3219 1.1739 -2.2976>, rotation=<rot 83.0415 0 0>}]
}
constraints [
 {
  bone1Name=thigh_l
  bone2Name=pelvis
  swing1LimitDegrees=45
  swing1Motion=Free
  swing2LimitDegrees=45
  swing2Motion=Free
  twistLimitDegrees=45
  twistMotion=Free
 }
 {
  bone1Name=spine_01
  bone2Name=pelvis
  swing1LimitDegrees=45
  swing1Motion=Free
  swing2LimitDegrees=45
  swing2Motion=Free
  twistLimitDegrees=45
  twistMotion=Free
 }
 {
  bone1Name=thigh_r
  bone2Name=pelvis
  swing1LimitDegrees=45
  swing1Motion=Free
  swing2LimitDegrees=45
  swing2Motion=Free
  twistLimitDegrees=45
  twistMotion=Free
 }
 {bone1Name=dyn_L_beltRing_front, bone2Name=pelvis, swing1LimitDegrees=20, swing2LimitDegrees=30, twistLimitDegrees=20}
 {bone1Name=dyn_L_beltRing_back, bone2Name=pelvis, swing1LimitDegrees=0, swing1Motion=Locked, swing2LimitDegrees=15, twistLimitDegrees=5}
 {bone1Name=dyn_L_hip_pivot_D, bone2Name=pelvis, swing1LimitDegrees=10, swing2LimitDegrees=60, twistLimitDegrees=20, twistMotion=Locked}
 {
  bone1Name=dyn_L_belt_B
  bone2Name=dyn_L_hip_pivot_D_02
  swing1LimitDegrees=20
  swing2LimitDegrees=20
  twistLimitDegrees=10
  twistMotion=Locked
 }
 {
  bone1Name=dyn_L_belt_D
  bone2Name=dyn_L_hip_pivot_D_02
  swing1LimitDegrees=30
  swing2LimitDegrees=20
  twistLimitDegrees=45
  twistMotion=Locked
 }
 {bone1Name=dyn_L_hip_pivot_D_02, bone2Name=dyn_L_hip_pivot_D, swing1LimitDegrees=5, swing2LimitDegrees=5, twistLimitDegrees=5}
]
